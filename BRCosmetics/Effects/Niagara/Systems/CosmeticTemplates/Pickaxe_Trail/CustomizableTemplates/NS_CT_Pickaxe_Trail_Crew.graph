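# Customizable Niagara template for the Crew pickaxe swing trail cosmetic, pairing socket-spawned particles with animation-trail ribbons.
# NE_Particles spawns five groups by distance travelled, spaced per swing section, with a visibility tag varying forces across sprite, mesh and ribbon particles.
# NE_Trails builds ribbons from user width and lifetime, faded by system age.
# User Variant switches color treatments, including pulsing ColorIn/ColorOut, while time-of-day tints both emitters.
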NiagaraSystem /BRCosmetics/Effects/Niagara/Systems/CosmeticTemplates/Pickaxe_Trail/CustomizableTemplates/NS_CT_Pickaxe_Trail_Crew
dependencies {
 usedDataInterfaces [
  &NiagaraDataInterfaceCurve
  &NiagaraDataInterfaceSpriteRendererInfo
  &NiagaraDataInterfaceColorCurve
  &NiagaraDataInterfaceSocketReader
 ]
 usedDynamicInputs [&SetBoolByFloatComparison, &Boolean_AndOperation]
 usedModules [
  &self:SystemSpawnScript.SystemScriptSource.SystemScriptGraph.NiagaraNodeAssignment_1.SetVariables_#1
  &SystemState
  &self:SystemSpawnScript.SystemScriptSource.SystemScriptGraph.NiagaraNodeAssignment_0.SetVariables_#2
  &self:SystemSpawnScript.SystemScriptSource.SystemScriptGraph.NiagaraNodeAssignment_2.SetVariables_#3
  &self:SystemSpawnScript.SystemScriptSource.SystemScriptGraph.NiagaraNodeAssignment_3.SetVariables_#4
  &self:SystemSpawnScript.SystemScriptSource.SystemScriptGraph.NiagaraNodeAssignment_5.SetVariables_#5
 ]
}
emitters {
 NE_Particles {
  emitterUpdateScript [
   &EmitterState(Life Cycle Mode=::NewEnumerator1, Loop Behavior=::NewEnumerator1, Loop Delay=0.05, Loop Duration=0.25, UseLoopDelay=true)
   &SpawnPerUnit(
    Module.Use Max Movement Threshold=false
    Module.Use Movement Tolerance=false
    Spawn Probability=1
    Spawn Spacing:NiagaraFloat=&MergeHelper_FloatSwitch(0=40, 1=40, 2=40, 3=40, Index=$User.SwingSectionIndex, Num=::NewEnumerator2)
    Use Spawn Probability=false
   )
   &SpawnPerUnit(
    Module.Use Max Movement Threshold=false
    Module.Use Movement Tolerance=false
    Spawn Group=1
    Spawn Probability=1
    Spawn Spacing:NiagaraFloat=&MergeHelper_FloatSwitch(0=40, 1=35, 2=45, 3=45, Index=$User.SwingSectionIndex, Num=::NewEnumerator2)
    Use Spawn Probability=false
   ) {module=SpawnPerUnit001}
   &SpawnPerUnit(
    Module.Use Max Movement Threshold=false
    Module.Use Movement Tolerance=false
    Spawn Group=2
    Spawn Probability=1
    Spawn Spacing:NiagaraFloat=&MergeHelper_FloatSwitch(0=35, 1=30, 2=40, 3=40, Index=$User.SwingSectionIndex, Num=::NewEnumerator2)
    Use Spawn Probability=false
   ) {module=SpawnPerUnit002}
   &SpawnPerUnit(
    Module.Use Max Movement Threshold=false
    Module.Use Movement Tolerance=false
    Spawn Group=3
    Spawn Spacing:NiagaraFloat=&MergeHelper_FloatSwitch(0=35, 1=30, 2=25, 3=30, Index=$User.SwingSectionIndex, Num=::NewEnumerator2)
    Use Spawn Probability=false
   ) {module=SpawnPerUnit003}
   &SpawnPerUnit(
    Module.Use Max Movement Threshold=false
    Module.Use Movement Tolerance=false
    Spawn Group=4
    Spawn Spacing:NiagaraFloat=&MergeHelper_FloatSwitch(0=40, 1=40, 2=40, 3=40, Index=$User.SwingSectionIndex, Num=::NewEnumerator2)
    Use Spawn Probability=false
   ) {module=SpawnPerUnit004}
  ]
  particleSpawnScript [
   &InitializeParticle(
    Color=&MergeHelper_ColorSwitch(
     0=&RandomRangeLinearColor(Minimum=<rgba 10 3.6667 0.5 1>, Maximum=<rgba 10 5.5 1 1>)
     1=<rgba 0.7 0.925 1 1>
     2=<rgba 0.7646 0.9519 1 1>
     3=<rgba 0 0 0 0>
     4=<rgba 0 0 0 0>
     5=&Multiply_LinearColorByFloat(Linear Color=<rgba 1 1 1 1>, Float=&FN_FrontEndFloatModifier(Gameplay=1, Front End=1.3))
     Index=$Emitter.SpawnGroup
     Num=::NewEnumerator4
    )
    Lifetime=&MergeHelper_FloatSwitch(
     0=&RandomRangeFloat(Minimum=0.3, Maximum=0.45)
     1=&RandomRangeFloat(Minimum=0.3, Maximum=0.45)
     2=&RandomRangeFloat(Minimum=0.3, Maximum=0.45)
     3=&RandomRangeFloat(Minimum=0.3, Maximum=0.45)
     4=0.25
     5=&RandomRangeFloat(Minimum=0.45, Maximum=0.75)
     Index=$Emitter.SpawnGroup
     Num=::NewEnumerator4
    )
    Mesh Scale Mode=::NewEnumerator3
    Mesh Uniform Scale=&Multiply_Float(
     A=&RandomRangeFloat(Minimum=0.12, Maximum=0.195)
     B=&FN_FrontEndFloatModifier(Gameplay=1, Front End=0.6)
    )
    Ribbon Width=&MaskFloat(Mask Source=1, Mask Parameter=$Emitter.SpawnGroup, Mask=5, Masked Float=0, Passthrough Float=0.75)
    Ribbon Width Mode=::NewEnumerator1
    Sprite Rotation Mode=::NewEnumerator2
    Sprite Size Mode=::NewEnumerator3
    Uniform Sprite Size=&Multiply_Float(
     A=&MergeHelper_FloatSwitch(
      0=&RandomRangeFloat(Minimum=4.5, Maximum=8.5)
      1=&RandomRangeFloat(Minimum=6, Maximum=11)
      2=&RandomRangeFloat(Minimum=6, Maximum=11)
      3=&RandomRangeFloat(Minimum=6, Maximum=11)
      4=&RandomRangeFloat(Minimum=10, Maximum=15)
      Index=$Emitter.SpawnGroup
      Num=::NewEnumerator3
     )
     B=&FN_FrontEndFloatModifier(Gameplay=1, Front End=0.6)
    )
    UsePositionOffset=false
    Write Material Random=true
   )
   &SocketLocation(
    Interpolate Socket Sampling=true
    Orientation Sampling=::NewEnumerator1
    SamplePosOffset=false
    Socket Index=1
    Socket Reader=$System.Socket Reader
    Socket Sampling Mode=::NewEnumerator3
    VelocitySpeedLimit=false
   )
   &ShapeLocation(MaskSpawnGroup=false, Override Seed=false, Shape Origin=$Output.SocketLocation.SampledPosition, Sphere Radius=10, UseOwnerScale=true)
   &AddVelocity(
    Module.Use Falloff Distance=false
    Override Seed=false
    Velocity Mode=::NewEnumerator1
    Velocity Origin=$Output.SocketLocation.SampledPosition
    Velocity Speed=&MaskFloat(
     Mask Source=1
     Mask Parameter=$Emitter.SpawnGroup
     Mask=5
     Masked Float=&RandomRangeFloat(Minimum=50, Maximum=120)
     Passthrough Float=&RandomRangeFloat(Minimum=0, Maximum=50)
    )
   )
   /SetVariables(Particles.VisibilityTag=$Emitter.SpawnGroup) {isSetParameters=true}
   /SetVariables(Particles.RibbonID=/Niagara/DynamicInputs/NiagaraID/MakeNiagaraIDFromInt(
     Index:NiagaraInt32=/Niagara/DynamicInputs/Divide/Divide_Int(A:NiagaraInt32=&ReturnParticleIDAttributes(), B:NiagaraInt32=36)
     Acquire Tag:NiagaraInt32=0
    )) {isSetParameters=true}
   &InheritVelocity(
    Inherited Velocity Amount Scale=&MaskVector(
     Mask Source=1
     Mask Parameter=$Emitter.SpawnGroup
     Mask=5
     Masked Vector=&VectorFromFloat(Value=&RandomRangeFloat(Minimum=0.1, Maximum=1))
     Passthrough Vector=<1 1 1>
    )
    Inherited Velocity Speed Limit:NiagaraFloat=&MergeHelper_FloatSwitch(
     0=0
     1=&RandomRangeFloat(Minimum=350, Maximum=600)
     2=0
     3=0
     4=0
     5=200
     Index=$Emitter.SpawnGroup
     Num=::NewEnumerator4
    )
    Speed Limit=true
    UseSpeedThreshold=false
    Velocity Source=$Output.SocketLocation.SampledVelocity
   )
  ]
  particleUpdateScript [
   &ParticleState()
   &SubUVAnimation(
    Output Alternate Renderer Binding=true
    Play Rate:NiagaraFloat=&RandomRangeFloat(Minimum=0.5, Maximum=1)
    Playback Scale Curve=DataInterface() {propertyValues {Curve {keys [
        {interpMode=RCIM_Cubic, tangentMode=RCTM_User, time=0.25, value=1}
        {arriveTangent=-1.2656, interpMode=RCIM_Cubic, leaveTangent=-1.2656, tangentMode=RCTM_User, time=1, value=0}
       ]}}}
    Random Start Frame=true
    "Sprite Renderer "=DataInterface() {propertyValues {SpriteRenderer=&self:NE_Particles_0.NiagaraSpriteRendererProperties_4}}
    SubUV Animation Mode:ENiagara_SubUVLookupModeV2=::NewEnumerator3
    UseEndFrame=true
    UsePlayRateScale=false
    UseStartFrame=true
   ) {module=SubUVAnimation001}
   &ScaleMeshSize(Scale Factor=&VectorFromFloat(Value=&FloatFromCurve(FloatCurve=DataInterface() {propertyValues {Curve {keys [<at 0 1 RCIM_Cubic>,<at 1 0 RCIM_Cubic>]}}}, Scale Curve=2)))
   &SpriteRotationRate(Rotation Rate=&FloatFromCurve(
     FloatCurve=DataInterface() {propertyValues {Curve {keys [{interpMode=RCIM_Cubic, tangentMode=RCTM_User, time=0, value=1}, {interpMode=RCIM_Cubic, tangentMode=RCTM_User, time=1, value=0}]}}}
     Scale Curve=&Multiply_Float(
      A=&MaskFloat(
       Mask Source=1
       Mask Parameter=$Particles.VisibilityTag
       Mask=5
       Masked Float=&Multiply_Float(
        A=&RandomRangeFloat(Minimum=1, Maximum=0.35)
        B=&FloatFromCurve(
         FloatCurve=DataInterface() {propertyValues {Curve {keys [{arriveTangent=-2, interpMode=RCIM_Cubic, time=0.15, value=1}, {interpMode=RCIM_Cubic, leaveTangent=-8.0089, time=0.5, value=0}]}}}
         Scale Curve=800
        )
       )
       Passthrough Float=&RandomRangeFloat(Minimum=100, Maximum=200)
      )
      B=&UniformAOrBFloat(A=-1, Evaluation Type=::NewEnumerator0, Recalculate Random Each Loop=true, Override Seed=false)
     )
    ))
   &Drag(
    Drag=&Multiply_Float(
     A=&UniformRangedFloat(Minimum=1.2, Maximum=1.5)
     B=&FloatFromCurve(
      FloatCurve=DataInterface() {propertyValues {Curve {keys [
          {arriveTangent=-5.3551, interpMode=RCIM_Cubic, leaveTangent=-5.3551, tangentMode=RCTM_User, time=0, value=0}
          {arriveTangent=-138.8873, interpMode=RCIM_Cubic, leaveTangent=-138.8873, tangentMode=RCTM_User, time=0.05, value=7}
          <at 0.2 0.5>
         ]}}}
      Scale Curve=4
     )
    )
    Use Linear Drag=true
    Use Rotational Drag=false
   )
   &AccelerationForce(Acceleration=&MaskVector(
     Mask Source=1
     Mask Parameter=$Particles.VisibilityTag
     Mask=5
     Masked Vector=&RandomRangeVector(Minimum=<0 0 -600>, Maximum=<0 0 -960>)
     Passthrough Vector=&RandomRangeVector(Minimum=<0 0 -150>, Maximum=<0 0 -200>)
    ))
   &CurlNoiseForce(
    Bias Noise=false
    Noise Frequency=30
    Noise Strength=&MergeHelper_FloatSwitch(0=200, 1=0, 2=200, 3=300, 4=200, Index=$Particles.VisibilityTag, Num=::NewEnumerator3)
    Pan Noise=false
   )
   &ScaleSpriteSize(Uniform Curve Sprite Scale=DataInterface() {propertyValues {Curve {keys [
        {arriveTangent=-1.3862, interpMode=RCIM_Cubic, time=0, value=1}
        {arriveTangent=-1.8417, interpMode=RCIM_Cubic, leaveTangent=-1.8417, tangentMode=RCTM_User, time=1, value=0}
       ]}}})
   &ScaleColor(
    Scale Alpha=&Multiply_Float(
     A=&FloatFromCurve(FloatCurve=DataInterface() {propertyValues {Curve {keys [<at 0.3 1>,<at 1 0>]}}})
     B=&FNTimeOfDayScalar(Night Scale=0.5, Morning Scale=0.75, Evening Scale=0.75)
    )
    Scale RGB=&MakeVectorFromLinearColorRGB(Color=$System.TOD_Color)
    ScaleA=true
    ScaleRGB=true
   )
   /SetVariables(
    Particles.GlowColor=&MakeLinearColorFromVectorAndFloat(
     "Vector (RGB)"=&MakeVectorFromLinearColorRGB(Color=&MergeHelper_ColorSwitch(
       0=&Multiply_LinearColorByFloat(
        Linear Color=&ColorFromCurve(
         ColorCurve=DataInterface() {propertyValues {
           AlphaCurve {keys [<at 0 1 RCIM_Cubic>]}
           BlueCurve {keys [
             <at 0 1>
             <at 0.2788 1>
             {arriveTangent=0.4199, interpMode=RCIM_Cubic, leaveTangent=0.4199, time=0.5087, value=0.9509}
             <at 0.7781 1.3267>
             <at 1 1>
            ]}
           GreenCurve {keys [
             <at 0 0.6539>
             <at 0.2788 0.2522>
             {arriveTangent=1.0874, interpMode=RCIM_Cubic, leaveTangent=1.0874, time=0.5087, value=0.5156}
             <at 0.7781 1.5>
             <at 1 0.6539>
            ]}
           RedCurve {keys [
             <at 0 0.4301>
             <at 0.2788 0.613>
             {arriveTangent=-0.1655, interpMode=RCIM_Cubic, leaveTangent=-0.1655, time=0.5087, value=1}
             <at 0.7781 0.3013>
             <at 1 0.4301>
            ]}
          }}
         CurveIndex=&Frac_Float(Float=&Add_Float(A=$Particles.NormalizedAge, B=&RandomRangeFloat(Minimum=0, Maximum=-0.5)))
        )
        Float=&RandomRangeFloat(Minimum=0.2, Maximum=0.5)
       )
       1=&Multiply_LinearColorByFloat(
        Linear Color=&ColorFromCurve(
         ColorCurve=DataInterface() {propertyValues {
           AlphaCurve {keys [<at 0 1 RCIM_Cubic>]}
           BlueCurve {keys [
             <at 0.0284 0.8677>
             {arriveTangent=-0.4812, interpMode=RCIM_Cubic, leaveTangent=-0.4812, time=0.4737, value=0.8646}
             <at 0.7925 0.5>
             <at 0.9918 0.8646>
            ]}
           GreenCurve {keys [
             <at 0.0284 1>
             {arriveTangent=-1.2221, interpMode=RCIM_Cubic, leaveTangent=-1.2221, time=0.4737, value=0.1067}
             <at 0.7925 0.0662>
             <at 0.9918 0.7183>
            ]}
           RedCurve {keys [
             <at 0.0284 0>
             {arriveTangent=0.4161, interpMode=RCIM_Cubic, leaveTangent=0.4161, time=0.4737, value=0.231}
             <at 0.7925 0.3179>
             <at 0.9918 0>
            ]}
          }}
         CurveIndex=&Frac_Float(Float=&Add_Float(A=$Particles.NormalizedAge, B=&RandomRangeFloat(Minimum=0, Maximum=1)))
        )
        Float=&RandomRangeFloat(Minimum=0.2, Maximum=0.5)
       )
       2=?
       3=?
       Index=/Niagara/DynamicInputs/Helpers/MaskInt(
        Mask Source:NiagaraInt32=1
        Mask Parameter:NiagaraInt32=&MakeIntFromFloat(Float=$User.Variant)
        Mask:NiagaraInt32=5
        Masked Int:NiagaraInt32=1
        Passthrough Int:NiagaraInt32=0
       )
       Num=::NewEnumerator0
      ))
     "Float (Alpha)"=1
    )
    Particles.RibbonColor:LinearColor=&MakeLinearColorFromVectorAndFloat(
     "Vector (RGB)"=&MakeVectorFromLinearColorRGB(Color=&Multiply_LinearColorByFloat(
       Linear Color=&ColorFromCurve(
        ColorCurve=DataInterface() {propertyValues {
          AlphaCurve {keys [<at 0 1 RCIM_Cubic>]}
          BlueCurve {keys [<at 0.0378 0.8677>, {arriveTangent=-0.0033, interpMode=RCIM_Cubic, time=0.3642, value=0.8646}, <at 0.7623 0.5>, <at 0.9918 0.8646>]}
          GreenCurve {keys [<at 0.0378 1>, {arriveTangent=-0.2924, interpMode=RCIM_Cubic, time=0.3642, value=0.1067}, <at 0.7623 0.0662>, <at 0.9918 0.7183>]}
          RedCurve {keys [<at 0.0378 0>,<at 0.3642 0.231 RCIM_Cubic>,<at 0.7623 0.3179>,<at 0.9918 0>]}
         }}
        CurveIndex=&Frac_Float(Float=&Add_Float(
          A=&Multiply_Float(A=$Particles.NormalizedAge, B=0.3)
          B=&RandomRangeFloat(Minimum=0, Maximum=1)
         ))
       )
       Float=&RandomRangeFloat(Minimum=0.35, Maximum=0.6)
      ))
     "Float (Alpha)"=&FloatFromCurve(
      FloatCurve=DataInterface() {propertyValues {Curve {keys [
          {interpMode=RCIM_Cubic, tangentMode=RCTM_Break, time=0.5, value=1}
          {arriveTangent=-0.6201, interpMode=RCIM_Cubic, tangentMode=RCTM_Break, time=1, value=0}
         ]}}}
      Scale Curve=1.1
     )
    )
   ) {isSetParameters=true}
   &SolveForcesAndVelocity(Clamp Velocity=false, Limit Acceleration=false)
  ]
  rendererClasses [
   &NiagaraSpriteRendererProperties
   &NiagaraMeshRendererProperties
   &NiagaraRibbonRendererProperties
  ]
  renderers [
   {rendererClass=&NiagaraSpriteRendererProperties}
   {rendererClass=&NiagaraSpriteRendererProperties, rendererIndex=1}
   {rendererClass=&NiagaraMeshRendererProperties, rendererIndex=2}
   {rendererClass=&NiagaraSpriteRendererProperties, rendererIndex=3}
   {rendererClass=&NiagaraSpriteRendererProperties, rendererIndex=4}
   {rendererClass=&NiagaraRibbonRendererProperties, rendererIndex=5}
  ]
 }
 NE_Trails {
  emitterSpawnScript [&FN_RibbonIDOffsetUtility()]
  emitterUpdateScript [
   &EmitterState()
   &FN_SpawnAnimTrail()
  ]
  particleSpawnScript [
   &FN_InitializeAnimTrailBoneIndecies()
   &FN_InitializeAnimTrail(AnimTrailWidth=$User.TrailWidth1, Lifetime=$User.TrailLifetime1)
  ]
  particleUpdateScript [
   &SolveForcesAndVelocity(Clamp Velocity=false, Limit Acceleration=false)
   &FN_UpdateAnimTrail() {module=FN_UpdateAnimTrail001}
   &ScaleColor(
    Scale Alpha=&Multiply_Float(
     A=&FloatFromCurve(
      FloatCurve=DataInterface() {propertyValues {Curve {
         defaultValue=1
         keys [{interpMode=RCIM_Cubic, tangentMode=RCTM_User, time=0, value=1}, {interpMode=RCIM_Cubic, tangentMode=RCTM_User, time=1, value=0}]
        }}}
      Scale Curve=&MergeHelper_FloatSwitch(0=1, 1=1, 2=0.7, 3=1, 4=1, 5=1, Index=&MakeIntFromFloat(Float=$User.Variant), Num=::NewEnumerator4)
     )
     B=&FloatFromCurve(
      FloatCurve=DataInterface() {propertyValues {Curve {keys [
          <at 0 0 RCIM_Cubic>
          <at 0.05 1>
          {arriveTangent=-4, interpMode=RCIM_Cubic, leaveTangent=0.2823, tangentMode=RCTM_Break, time=0.2, value=1}
          {arriveTangent=0.3821, interpMode=RCIM_Cubic, leaveTangent=-7.0431, tangentMode=RCTM_Break, time=0.3, value=0}
         ]}}}
      CurveIndex=$System.Age
     )
    )
    Scale RGB=&Multiply_Vector(
     A=&MakeVectorFromLinearColorRGB(Color=&MergeHelper_ColorSwitch(
       0=$System.TOD_Color
       1=&FNTimeOfDayColor(Night Color=$User.NightTOD, Morning Color=<rgba 1 1 1 1>, Day Color=<rgba 1 1 1 1>, Evening Color=$User.EveningTOD)
       2=?
       3=?
       Index=&MakeIntFromFloat(Float=$User.UseCustomTOD)
       Num=::NewEnumerator0
      ))
     B=&MakeVectorFromLinearColorRGB(Color=&MergeHelper_ColorSwitch(
       0=<rgba 1 1 1 1>
       1=<rgba 1 1 1 1>
       2=<rgba 1 1 1 1>
       3=<rgba 1 1 1 1>
       4=&LerpLinearColors(
        StartColor=$User.ColorIn
        EndColor=$User.ColorOut
        LerpFactor=&Waveform("Waveform [1]"=::NewEnumerator8, "[1] Frequency"=10, "[1] Amplitude Min/Max"=<0 1>)
       )
       5=&LerpLinearColors(
        StartColor=$User.ColorIn
        EndColor=$User.ColorOut
        LerpFactor=&Waveform("Waveform [1]"=::NewEnumerator8, "[1] Frequency"=10, "[1] Amplitude Min/Max"=<0 1>)
       )
       Index=&MakeIntFromFloat(Float=$User.Variant)
       Num=::NewEnumerator4
      ))
    )
    ScaleA=true
    ScaleRGB=true
   ) {module=ScaleColor001}
   /SetVariables(Particles.DynamicParametersAlt3:Vector4f=&MakeVector4(
     X=0
     Y=0
     Z=0
     W=&Multiply_Float(
      A=&FloatFromCurve(
       FloatCurve=DataInterface() {propertyValues {Curve {keys [
           <at 0 0 RCIM_Cubic>
           <at 0.05 1>
           {arriveTangent=-4, interpMode=RCIM_Cubic, leaveTangent=0.2823, tangentMode=RCTM_Break, time=0.2, value=1}
           {arriveTangent=0.3821, interpMode=RCIM_Cubic, leaveTangent=-7.0431, tangentMode=RCTM_Break, time=0.3, value=0}
          ]}}}
       CurveIndex=$System.Age
      )
      B=0.2
     )
    )) {isSetParameters=true}
  ]
  rendererClasses [&NiagaraRibbonRendererProperties]
  renderers [
   {rendererClass=&NiagaraRibbonRendererProperties}
   {rendererClass=&NiagaraRibbonRendererProperties, rendererIndex=1}
   {rendererClass=&NiagaraRibbonRendererProperties, rendererIndex=2}
   {rendererClass=&NiagaraRibbonRendererProperties, rendererIndex=3}
   {rendererClass=&NiagaraRibbonRendererProperties, rendererIndex=4}
   {rendererClass=&NiagaraRibbonRendererProperties, rendererIndex=5}
   {rendererClass=&NiagaraRibbonRendererProperties, rendererIndex=6}
   {rendererClass=&NiagaraRibbonRendererProperties, rendererIndex=7}
   {rendererClass=&NiagaraRibbonRendererProperties, rendererIndex=8}
   {rendererClass=&NiagaraRibbonRendererProperties, rendererIndex=9}
   {rendererClass=&NiagaraRibbonRendererProperties, rendererIndex=10}
   {rendererClass=&NiagaraRibbonRendererProperties, rendererIndex=11}
   {rendererClass=&NiagaraRibbonRendererProperties, rendererIndex=12}
   {rendererClass=&NiagaraRibbonRendererProperties, rendererIndex=13}
  ]
 }
}
summary {
 systemName=NS_CT_Pickaxe_Trail_Crew
 emitters [
  {
   emitterName=NE_Particles
   rendererClasses [
    &NiagaraSpriteRendererProperties
    &NiagaraMeshRendererProperties
    &NiagaraRibbonRendererProperties
   ]
  }
  {emitterName=NE_Trails, rendererClasses [&NiagaraRibbonRendererProperties]}
 ]
 userVariables [
  User.bIsFrontEnd {type=NiagaraBool}
  User.bIsFrontEndPreview {type=NiagaraBool}
  User.ColorIn {type=LinearColor, description="", defaultValue {struct=LinearColor, value=<rgba 0.7758 0.3613 1 1>}}
  User.ColorOut {type=LinearColor, description="", defaultValue {struct=LinearColor, value=<rgba 0.1746 0.7835 1 1>}}
  User.EveningTOD {type=LinearColor, description="", defaultValue {struct=LinearColor, value=<rgba 1 1 1 1>}}
  User.MITrail {
   type=MaterialInterface
   description=""
   defaultValue {struct=NiagaraExt_VariableValue_Object, value {object=null, objectClass=/Script/Engine.MaterialInterface}}
  }
  User.MITrailGalaxy {
   type=MaterialInterface
   description=""
   defaultValue {struct=NiagaraExt_VariableValue_Object, value {object=null, objectClass=&MaterialInterface}}
  }
  User.NightTOD {type=LinearColor, description="", defaultValue {struct=LinearColor, value=<rgba 1 1 1 1>}}
  User.RibbonIDOffset {type=NiagaraInt32}
  User.SwingSectionIndex {type=NiagaraInt32}
  User.ToggleScreen {type=NiagaraFloat, defaultValue=0, description=""}
  User.TrailLifetime1 {type=NiagaraFloat, defaultValue=0.3, description=""}
  User.TrailWidth1 {type=NiagaraFloat, defaultValue=1, description=""}
  User.UseCustomTOD {type=NiagaraFloat, defaultValue=0, description=""}
  User.Variant {type=NiagaraFloat, defaultValue=0, description=""}
 ]
}
