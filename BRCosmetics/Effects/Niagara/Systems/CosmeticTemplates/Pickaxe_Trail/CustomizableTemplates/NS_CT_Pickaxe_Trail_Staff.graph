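# Customizable Niagara template for a staff pickaxe's melee trail, exposing user materials, colors and widths for two trails.
# One emitter spawns two anim-trail ribbons between bone or socket pairs selected by User.IsR2.
# Spawn group picks each trail's color and width and sets the visibility tag for the two ribbon renderers.
# Color is tinted by time of day, then short-lived particles fade and taper through curves.
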NiagaraSystem /BRCosmetics/Effects/Niagara/Systems/CosmeticTemplates/Pickaxe_Trail/CustomizableTemplates/NS_CT_Pickaxe_Trail_Staff
dependencies {
 usedDataInterfaces [
  &NiagaraDataInterfaceSkeletalMesh
  &NiagaraDataInterfaceColorCurve
  &NiagaraDataInterfaceCurve
  &NiagaraDataInterfaceSocketReader
 ]
 usedDynamicInputs [&SetBoolByFloatComparison]
 usedModules [
  &self:SystemSpawnScript.SystemScriptSource.SystemScriptGraph.NiagaraNodeAssignment_1.SetVariables_#1
  &self:SystemSpawnScript.SystemScriptSource.SystemScriptGraph.NiagaraNodeAssignment_0.SetVariables_#2
  &SystemState
 ]
}
emitters {NE_Staff_Trail {
  emitterSpawnScript [&FN_RibbonIDOffsetUtility()]
  emitterUpdateScript [
   &EmitterState()
   &FN_SpawnAnimTrail(Number of Anim Trails=2)
  ]
  particleSpawnScript [
   &FN_InitializeAnimTrailBoneIndecies(
    First Bone or Socket Index:NiagaraInt32=&MakeCustomIntFromBool(Bool=$User.IsR2, True INT=2, False INT=0)
    Second Bone or Socket Index:NiagaraInt32=&MakeCustomIntFromBool(Bool=$User.IsR2, True INT=3, False INT=1)
   )
   &FN_InitializeAnimTrailBoneIndecies(
    AnimTrail Selection=1
    First Bone or Socket Index:NiagaraInt32=&MakeCustomIntFromBool(Bool=$User.IsR2, True INT=2, False INT=0)
    Second Bone or Socket Index:NiagaraInt32=&MakeCustomIntFromBool(Bool=$User.IsR2, True INT=3, False INT=1)
   ) {module=FN_InitializeAnimTrailBoneIndecies001}
   /SetVariables(Particles.VisibilityTag=&SpawnGroup()) {isSetParameters=true}
   &FN_InitializeAnimTrail(
    AnimTrailWidth=&MergeHelper_FloatSwitch(0=$User.TrailWidth1, 1=$User.TrailWidth2, 2=?, 3=?, Index=&SpawnGroup(), Num=::NewEnumerator0)
    Color=&MergeHelper_ColorSwitch(0=$User.TrailColor1, 1=$User.TrailColor2, 2=?, 3=?, Index=&SpawnGroup(), Num=::NewEnumerator0)
    Lifetime=0.15
   )
   &Color(Color=&MultiplyLinearColor(
     Linear Color=$Particles.Initial.Color
     Scale Factor=&MakeVector4_FromLinearColor(LinearColor=&FN_Select_Color_from_Time_Of_Day(Morning=<rgba 0.6719 0.361 0.5226 0.8>, Day=<rgba 1 1 1 1>, Evening=<rgba 1 0.6413 0.7956 0.7>, Night=<rgba 0.3188 0.3311 0.5885 1>))
    )) {module=Color001}
  ]
  particleUpdateScript [
   &FN_UpdateAnimTrail(Locked Particle RibbonWidth:NiagaraFloat=$Particles.RibbonWidth) {module=FN_UpdateAnimTrail001}
   &ScaleColor(
    Linear Color Curve=DataInterface() {propertyValues {
      AlphaCurve {keys [<at 0 0.7>,<at 0.038 1>,<at 0.1575 0.35>,<at 0.4643 0.06>,<at 0.9419 0>]}
      BlueCurve {keys [<at 0 1>,<at 0.871 0.4271>]}
      GreenCurve {keys [<at 0 1>,<at 0.871 0.4271>]}
      RedCurve {keys [<at 0 1>,<at 0.871 0.4271>]}
     }}
    Scale Mode=::NewEnumerator2
   )
   &ScaleRibbonWidth(Ribbon Width Scale=&FloatFromCurve(FloatCurve=DataInterface() {propertyValues {Curve {keys [
         {interpMode=RCIM_Cubic, tangentMode=RCTM_User, time=0, value=1}
         {arriveTangent=-3.1724, interpMode=RCIM_Cubic, leaveTangent=-3.1724, tangentMode=RCTM_User, time=1, value=0}
        ]}}}))
   &SolveForcesAndVelocity(Clamp Velocity=false, Limit Acceleration=false)
  ]
  rendererClasses [&NiagaraRibbonRendererProperties]
  renderers [
   {rendererClass=&NiagaraRibbonRendererProperties}
   {rendererClass=&NiagaraRibbonRendererProperties, rendererIndex=1}
  ]
 }}
summary {
 systemName=NS_CT_Pickaxe_Trail_Staff
 emitters [{emitterName=NE_Staff_Trail, rendererClasses [&NiagaraRibbonRendererProperties]}]
 userVariables [
  User.IsR2 {type=NiagaraBool}
  User.MITrail1 {
   type=MaterialInterface
   description=""
   defaultValue {
    struct=NiagaraExt_VariableValue_Object
    value {object=/Game/Effects/Fort_Effects/Materials/Melee/M_Melee_Trail_Master, objectClass=/Script/Engine.MaterialInterface}
   }
  }
  User.MITrail2 {
   type=MaterialInterface
   description=""
   defaultValue {
    struct=NiagaraExt_VariableValue_Object
    value {object=&M_Melee_Trail_Master, objectClass=&MaterialInterface}
   }
  }
  User.RibbonIDOffset {type=NiagaraInt32}
  User.SwingSectionIndex {type=NiagaraInt32}
  User.TrailColor1 {type=LinearColor, description="", defaultValue {struct=LinearColor, value=<rgba 1 1 1 1>}}
  User.TrailColor2 {type=LinearColor, description="", defaultValue {struct=LinearColor, value=<rgba 0.8632 0.8632 0.8632 0.5>}}
  User.TrailWidth1 {type=NiagaraFloat, defaultValue=1, description=""}
  User.TrailWidth2 {type=NiagaraFloat, defaultValue=1, description=""}
  User.UseTrail2 {type=NiagaraFloat, defaultValue=0, description=""}
 ]
}
